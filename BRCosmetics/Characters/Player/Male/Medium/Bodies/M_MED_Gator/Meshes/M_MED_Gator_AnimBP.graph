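# Post-process style AnimBP for the M_MED_Gator character body, copying the attached parent's pose and procedurally adjusting belt, shell and shoulder pad bones.
# Weighted constraints drive belts from pelvis/thigh and shells from belts.
# Event update reads ref-pose thigh and arm deltas as alphas for additive bone offsets, with a melee-weapon check rolling the left pad.
# Right-arm values also sample left-arm bones; the right pad's melee node is unconnected.
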AnimBlueprint /BRCosmetics/Characters/Player/Male/Medium/Bodies/M_MED_Gator/Meshes/M_MED_Gator_AnimBP
target_skeleton=&/M_MED_Gator_Skeleton
graphs [
 AnimGraph {nodes [
   "Constraint - dyn_shldr_pad_pivot_base_l (P)"(Alpha=1, ConstraintWeights_0=1) {
    alphaBoolBlend {blendOption=Linear}
    boneToModify {boneName=dyn_shldr_pad_pivot_base_l}
    constraintSetup [{targetBone {boneName=upperarm_twist_01_l}}]
    constraintWeights [1]
   }
   "Transform (Modify) Bone - Bone: dyn_shldr_pad_pivot_base_r"!(ComponentPose, bAlphaBoolEnabled:bool=Get isMeleeWeapon() -> (isMeleeWeapon:bool), Translation="0,0,0", Rotation="0,0,-20", Scale="1,1,1") -> (Pose) {
    alphaInputType=Bool
    rotationMode=BMM_Additive
    rotationSpace=BCS_BoneSpace
    translationMode=BMM_Additive
    translationSpace=BCS_BoneSpace
    alphaBoolBlend {blendOption=Linear}
    alphaScaleBiasClamp {inRange {max=70}}
    boneToModify {boneName=dyn_shldr_pad_pivot_base_r}
    rotation {roll=-20}
   }
   n1 = "Constraint - dyn_belt_a_base (PP)"!(
    ComponentPose=Local To Component(LocalPose=Copy Pose From Mesh() {bUseAttachedParent=true})
    Alpha=1
    ConstraintWeights_0=0.3148
    ConstraintWeights_1:float=0.6852
   ) -> (Pose) {
    alphaBoolBlend {blendOption=Linear}
    boneToModify {boneName=dyn_belt_a_base}
    constraintSetup [{targetBone {boneName=pelvis}}, {targetBone {boneName=thigh_twist_01_l}}]
    constraintWeights [1,1]
   }
   n2 = "Constraint - dyn_belt_b_base (PP)"!(ComponentPose=@n1, Alpha=1, ConstraintWeights_0=0.4853, ConstraintWeights_1:float=0.5147) -> (Pose) {
    alphaBoolBlend {blendOption=Linear}
    boneToModify {boneName=dyn_belt_b_base}
    constraintSetup [{targetBone {boneName=pelvis}}, {targetBone {boneName=thigh_twist_01_l}}]
    constraintWeights [1,1]
   }
   n3 = "Constraint - dyn_belt_c_base (PPP)"!(ComponentPose=@n2, Alpha=1, ConstraintWeights_0=0.5826, ConstraintWeights_1:float=0.3662, ConstraintWeights_2:float=0.0512) -> (Pose) {
    alphaBoolBlend {blendOption=Linear}
    boneToModify {boneName=dyn_belt_c_base}
    constraintSetup [{targetBone {boneName=pelvis}}, {targetBone {boneName=thigh_volume_l}}, {targetBone {boneName=thigh_twist_01_l}}]
    constraintWeights [1,1,1]
   }
   n4 = Get legUp_L() -> (legUp_L:double)
   n5 = "Transform (Modify) Bone - Bone: dyn_belt_a_base"(ComponentPose=@n3, Alpha=@n4, Translation="-1,3,1", Rotation="0,50,0", Scale="1,1,1") {
    rotationMode=BMM_Additive
    rotationSpace=BCS_BoneSpace
    translationMode=BMM_Additive
    translationSpace=BCS_BoneSpace
    alphaBoolBlend {blendOption=Linear}
    alphaScaleBiasClamp {bMapRange=true, interpSpeedDecreasing=1, interpSpeedIncreasing=1, inRange {max=0, min=-100}, outRange {max=0, min=1}}
    boneToModify {boneName=dyn_belt_a_base}
    rotation {yaw=50}
    translation=<-1 3 1>
   }
   n6 = "Transform (Modify) Bone - Bone: dyn_belt_b_base"(ComponentPose=@n5, Alpha=@n4, Translation="-6,0,2", Rotation="30.000,20.000,-5.000", Scale="1,1,1") {
    rotationMode=BMM_Additive
    rotationSpace=BCS_BoneSpace
    translationMode=BMM_Additive
    translationSpace=BCS_BoneSpace
    alphaBoolBlend {blendOption=Linear}
    alphaScaleBiasClamp {bMapRange=true, interpSpeedDecreasing=1, interpSpeedIncreasing=1, inRange {max=0, min=-100}, outRange {max=0, min=1}}
    boneToModify {boneName=dyn_belt_b_base}
    rotation=<rot 30 19.9999 -5.0001>
    translation {x=-6, z=2}
   }
   n7 = "Transform (Modify) Bone - Bone: dyn_belt_c_base"(ComponentPose=@n6, Alpha=@n4, Translation="3,0,2", Rotation="5,0,0", Scale="1,1,1") {
    rotationMode=BMM_Additive
    rotationSpace=BCS_BoneSpace
    translationMode=BMM_Additive
    translationSpace=BCS_BoneSpace
    alphaBoolBlend {blendOption=Linear}
    alphaScaleBiasClamp {bMapRange=true, interpSpeedDecreasing=1, interpSpeedIncreasing=1, inRange {max=0, min=-100}, outRange {max=0, min=-1}}
    boneToModify {boneName=dyn_belt_c_base}
    rotation {pitch=5}
    translation {x=3, z=2}
   }
   n8 = "Constraint - dyn_shell_a_base (PP)"!(ComponentPose=@n7, Alpha=1, ConstraintWeights_0=0.5556, ConstraintWeights_1:float=0.4444) -> (Pose) {
    alphaBoolBlend {blendOption=Linear}
    boneToModify {boneName=dyn_shell_a_base}
    constraintSetup [{targetBone {boneName=dyn_belt_a}}, {targetBone {boneName=dyn_belt_b}}]
    constraintWeights [1,1]
   }
   n9 = "Constraint - dyn_shell_b_base (PP)"!(ComponentPose=@n8, Alpha=1, ConstraintWeights_0=0.3625, ConstraintWeights_1:float=0.6375) -> (Pose) {
    alphaBoolBlend {blendOption=Linear}
    boneToModify {boneName=dyn_shell_b_base}
    constraintSetup [{targetBone {boneName=dyn_belt_a}}, {targetBone {boneName=dyn_belt_b}}]
    constraintWeights [1,1]
   }
   n10 = "Constraint - dyn_shell_c_base (PPP)"!(ComponentPose=@n9, Alpha=1, ConstraintWeights_0=0.195, ConstraintWeights_1:float=0.7732, ConstraintWeights_2:float=0.0317) -> (Pose) {
    alphaBoolBlend {blendOption=Linear}
    boneToModify {boneName=dyn_shell_c_base}
    constraintSetup [{targetBone {boneName=dyn_belt_a}}, {targetBone {boneName=dyn_belt_b}}, {targetBone {boneName=dyn_belt_c}}]
    constraintWeights [1,1,1]
   }
   n11 = "Constraint - dyn_shell_d_base (PPP)"!(ComponentPose=@n10, Alpha=1, ConstraintWeights_0=0.0864, ConstraintWeights_1:float=0.8259, ConstraintWeights_2:float=0.0876) -> (Pose) {
    alphaBoolBlend {blendOption=Linear}
    boneToModify {boneName=dyn_shell_d_base}
    constraintSetup [{targetBone {boneName=dyn_belt_a}}, {targetBone {boneName=dyn_belt_b}}, {targetBone {boneName=dyn_belt_c}}]
    constraintWeights [1,1,1]
   }
   n12 = "Constraint - dyn_shell_e_base (PPP)"!(ComponentPose=@n11, Alpha=1, ConstraintWeights_0=0.0324, ConstraintWeights_1:float=0.7767, ConstraintWeights_2:float=0.1909) -> (Pose) {
    alphaBoolBlend {blendOption=Linear}
    boneToModify {boneName=dyn_shell_e_base}
    constraintSetup [{targetBone {boneName=dyn_belt_a}}, {targetBone {boneName=dyn_belt_b}}, {targetBone {boneName=dyn_belt_c}}]
    constraintWeights [1,1,1]
   }
   n13 = Get armUp_L() -> (armUp_L:double)
   n14 = "Transform (Modify) Bone - Bone: dyn_shldr_pad_pivot_base_l"(ComponentPose=@n12, Alpha=@n13, Translation="9,-2,1.5", Rotation="-10,0,0", Scale="1,1,1") {
    rotationMode=BMM_Additive
    rotationSpace=BCS_BoneSpace
    translationMode=BMM_Additive
    translationSpace=BCS_BoneSpace
    alphaBoolBlend {blendOption=Linear}
    alphaScaleBiasClamp {bMapRange=true, inRange {max=70}}
    boneToModify {boneName=dyn_shldr_pad_pivot_base_l}
    rotation {pitch=-10}
    translation=<9 -2 1.5>
   }
   n15 = "Transform (Modify) Bone - Bone: dyn_shldr_pad_pivot_base_l"(ComponentPose=@n14, Alpha=@n13, Translation="0,-1,0", Rotation="0,0,-30", Scale="1,1,1") {
    rotationMode=BMM_Additive
    rotationSpace=BCS_BoneSpace
    translationMode=BMM_Additive
    translationSpace=BCS_BoneSpace
    alphaBoolBlend {blendOption=Linear}
    alphaScaleBiasClamp {bMapRange=true, inRange {max=0, min=-25}, outRange {max=0, min=1}}
    boneToModify {boneName=dyn_shldr_pad_pivot_base_l}
    rotation {roll=-30}
    translation {y=-1}
   }
   n16 = "Transform (Modify) Bone - Bone: dyn_shldr_pad_pivot_base_l"(ComponentPose=@n15, Alpha=Get armFwd_L() -> (armFwd_L:double), Translation="0,0,1", Rotation="5,0,-15", Scale="1,1,1") {
    rotationMode=BMM_Additive
    rotationSpace=BCS_BoneSpace
    translationMode=BMM_Additive
    translationSpace=BCS_BoneSpace
    alphaBoolBlend {blendOption=Linear}
    alphaScaleBiasClamp {bMapRange=true, inRange {max=70}}
    boneToModify {boneName=dyn_shldr_pad_pivot_base_l}
    rotation {pitch=5, roll=-15}
    translation {z=1}
   }
   n17 = "Transform (Modify) Bone - Bone: dyn_shldr_pad_pivot_base_l"!(ComponentPose=@n16, bAlphaBoolEnabled:bool=Get isMeleeWeapon(), Translation="0,0,0", Rotation="0,0,-20", Scale="1,1,1") -> (Pose) {
    alphaInputType=Bool
    rotationMode=BMM_Additive
    rotationSpace=BCS_BoneSpace
    translationMode=BMM_Additive
    translationSpace=BCS_BoneSpace
    alphaBoolBlend {blendOption=Linear}
    alphaScaleBiasClamp {inRange {max=70}}
    boneToModify {boneName=dyn_shldr_pad_pivot_base_l}
    rotation {roll=-20}
   }
   n18 = Get armUp_R() -> (armUp_R:double)
   n19 = "Transform (Modify) Bone - Bone: dyn_shldr_pad_pivot_base_r"!(ComponentPose=@n17, Alpha:float=@n18, Translation="-9,2,-1.5", Rotation="-10,0,0", Scale="1,1,1") -> (Pose) {
    rotationMode=BMM_Additive
    rotationSpace=BCS_BoneSpace
    translationMode=BMM_Additive
    translationSpace=BCS_BoneSpace
    alphaBoolBlend {blendOption=Linear}
    alphaScaleBiasClamp {bMapRange=true, interpSpeedDecreasing=1, interpSpeedIncreasing=1, inRange {max=70}}
    boneToModify {boneName=dyn_shldr_pad_pivot_base_r}
    rotation {pitch=-10}
    translation=<-9 2 -1.5>
   }
   n20 = "Transform (Modify) Bone - Bone: dyn_shldr_pad_pivot_base_r"(ComponentPose=@n19, Alpha=@n18, Translation="0,1,0", Rotation="0,0,-30", Scale="1,1,1") {
    rotationMode=BMM_Additive
    rotationSpace=BCS_BoneSpace
    translationMode=BMM_Additive
    translationSpace=BCS_BoneSpace
    alphaBoolBlend {blendOption=Linear}
    alphaScaleBiasClamp {bMapRange=true, interpSpeedDecreasing=1, interpSpeedIncreasing=1, inRange {max=0, min=-25}, outRange {max=0, min=1}}
    boneToModify {boneName=dyn_shldr_pad_pivot_base_r}
    rotation {roll=-30}
    translation {y=1}
   }
   n21 = "Transform (Modify) Bone - Bone: dyn_shldr_pad_pivot_base_r"(ComponentPose=@n20, Alpha=Get armFwd_R() -> (armFwd_R:double), Translation="0,0,-1", Rotation="5,0,-15", Scale="1,1,1") {
    rotationMode=BMM_Additive
    rotationSpace=BCS_BoneSpace
    translationMode=BMM_Additive
    translationSpace=BCS_BoneSpace
    alphaBoolBlend {blendOption=Linear}
    alphaScaleBiasClamp {bMapRange=true, interpSpeedDecreasing=1, interpSpeedIncreasing=1, inRange {max=70}}
    boneToModify {boneName=dyn_shldr_pad_pivot_base_r}
    rotation {pitch=5, roll=-15}
    translation {z=-1}
   }
   Output Pose(Result=Component To Local(ComponentPose=@n21)) {layerGroup=None}
  ]}
 EventGraph {nodes [
   Event BlueprintUpdateAnimation()
   > n1 = Sequence!(execute) -> (then_0, then_1, then_2:Exec, then_3:Exec)
   > n2 = Sequence!(execute) -> (then_0, then_1, then_2, then_3, then_4:Exec)
   n3 = GetDeltaTransformFromRefPose(self=GetOwnerSkeletalMeshComponent(), BoneName=thigh_l, BaseName=pelvis)
   Set legUp_L(execute=@n2.then_1, legUp_L:double=BreakRotator(InRot=BreakTransform(InTransform=@n3).Rotation).Yaw) -> (Output_Get:double)
   n4 = GetOwnerSkeletalMeshComponent()
   n5 = BreakRotator(InRot=BreakTransform(InTransform=GetDeltaTransformFromRefPose(self=@n4, BoneName=upperarm_l, BaseName=clavicle_l)).Rotation)
   Set armUp_L(execute=@n2.then_2, armUp_L:double=@n5.Pitch) -> (Output_Get:double)
   > Set armFwd_L(armFwd_L:double=@n5.Yaw) -> (Output_Get:double)
   n6 = GetOwnerSkeletalMeshComponent()
   n7 = BreakRotator(InRot=BreakTransform(InTransform=GetDeltaTransformFromRefPose(self=@n6, BoneName=upperarm_l, BaseName=clavicle_l)).Rotation)
   Set armUp_R(execute=@n2.then_3, armUp_R:double=@n7.Pitch) -> (Output_Get:double)
   > Set armFwd_R(armFwd_R:double=@n7.Yaw) -> (Output_Get:double)
   n8 = Cast To FortPlayerAnimInstance(execute=@n1.then_1, Object=GetAnimInstance(self=GetOwnerSkeletalMeshComponent()))
   n9 = Get WeaponCoreAnim(self=@n8)
   > Set isMeleeWeapon(isMeleeWeapon:bool="Equal Exactly (String)"(A=Melee, B=Enum to String!(Enumerator:EFortWeaponCoreAnimation Enum=@n9) -> (ReturnValue:String))) -> (Output_Get:bool)
   TryGetPawnOwner()
  ]}
]
variables [
 armFwd_L {type=number}
 armFwd_R {type=number}
 armUp_L {type=number}
 armUp_R {type=number}
 isMeleeWeapon {type=boolean}
 legUp_L {type=number}
]
