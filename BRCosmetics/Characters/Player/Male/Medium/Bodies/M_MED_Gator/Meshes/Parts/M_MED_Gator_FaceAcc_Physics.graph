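# Physics asset for the face accessory mesh of the male medium Gator body cosmetic, covering secondary motion of the horns on its hood.
# Five left/right horn pairs and an upper middle horn each get a fitted capsule body, plus a tiny capsule on a matching base bone.
# Each horn is constrained only to its own base with tight swing and twist limits, so horns can wobble slightly as separate two-body chains.
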PhysicsAsset /BRCosmetics/Characters/Player/Male/Medium/Bodies/M_MED_Gator/Meshes/Parts/M_MED_Gator_FaceAcc_Physics
bodies {
 dyn_hood_horn_a_base_l [{shapeType=Capsule, length=0.5, radius=0.5}]
 dyn_hood_horn_a_base_r [{shapeType=Capsule, length=0.5, radius=0.5}]
 dyn_hood_horn_a_l [{shapeType=Capsule, length=3.9855, radius=2.3321, center=<0.5277 1.5837 0.0442>, rotation=<rot 0.2482 172.1739 -89.0737>}]
 dyn_hood_horn_a_r [{shapeType=Capsule, length=3.1506, radius=2.0825, center=<-0.3686 -2.0367 -0.147>, rotation=<rot 0 179.4389 -87.3648>}]
 dyn_hood_horn_b_base_l [{shapeType=Capsule, length=0.5, radius=0.5}]
 dyn_hood_horn_b_base_r [{shapeType=Capsule, length=0.5, radius=0.5}]
 dyn_hood_horn_b_l [{shapeType=Capsule, length=2.8526, radius=1.5415, center=<-0.3977 2.2081 0.0598>, rotation=<rot 0 178.6887 -86.0785>}]
 dyn_hood_horn_b_r [{shapeType=Capsule, length=3.0332, radius=1.5189, center=<0.0772 -1.4811 0.0662>, rotation=<rot 0 175.034 -83.5256>}]
 dyn_hood_horn_c_base_l [{shapeType=Capsule, length=0.5, radius=0.5}]
 dyn_hood_horn_c_base_r [{shapeType=Capsule, length=0.5, radius=0.5}]
 dyn_hood_horn_c_l [{shapeType=Capsule, length=2.5904, radius=1.5267, center=<0.4259 1.9122 -0.0338>, rotation=<rot 0 172.3493 -82.1448>}]
 dyn_hood_horn_c_r [{shapeType=Capsule, length=2.4147, radius=1.4621, center=<-0.1286 -2.1267 0.0263>, rotation=<rot 0 175.1759 -81.1672>}]
 dyn_hood_horn_d_base_l [{shapeType=Capsule, length=0.5, radius=0.5}]
 dyn_hood_horn_d_base_r [{shapeType=Capsule, length=0.5, radius=0.5}]
 dyn_hood_horn_d_l [{shapeType=Capsule, length=2.5192, radius=1.4068, center=<0.1466 1.7912 0.052>, rotation=<rot 0 175.3667 -85.7177>}]
 dyn_hood_horn_d_r [{shapeType=Capsule, length=2.099, radius=1.3257, center=<-0.1626 -1.8633 -0.1172>, rotation=<rot 0 -177.8138 -84.6682>}]
 dyn_hood_horn_e_base_l [{shapeType=Capsule, length=0.5, radius=0.5}]
 dyn_hood_horn_e_base_r [{shapeType=Capsule, length=0.5, radius=0.5}]
 dyn_hood_horn_e_l [{shapeType=Capsule, length=2.302, radius=1.244, center=<0.0149 2.2957 0.0777>, rotation=<rot 0 -179.1398 -86.5154>}]
 dyn_hood_horn_e_r [{shapeType=Capsule, length=2.4266, radius=1.2691, center=<0.1445 -2.9249 -0.1152>, rotation=<rot 0 178.911 -85.3348>}]
 dyn_hood_horn_upMid [{shapeType=Capsule, length=2.3681, radius=1.2813, center=<0.1054 2.3186 -0.0836>, rotation=<rot 0 -179.3388 -85.7188>}]
 dyn_hood_horn_upMid_base [{shapeType=Capsule, length=0.5, radius=0.5}]
}
constraints [
 {bone1Name=dyn_hood_horn_a_l, bone2Name=dyn_hood_horn_a_base_l, swing1LimitDegrees=5, swing2LimitDegrees=5, twistLimitDegrees=15}
 {bone1Name=dyn_hood_horn_b_l, bone2Name=dyn_hood_horn_b_base_l, swing1LimitDegrees=5, swing2LimitDegrees=5, twistLimitDegrees=15}
 {bone1Name=dyn_hood_horn_c_l, bone2Name=dyn_hood_horn_c_base_l, swing1LimitDegrees=5, swing2LimitDegrees=5, twistLimitDegrees=15}
 {bone1Name=dyn_hood_horn_d_l, bone2Name=dyn_hood_horn_d_base_l, swing1LimitDegrees=10, swing2LimitDegrees=5, twistLimitDegrees=15}
 {bone1Name=dyn_hood_horn_e_l, bone2Name=dyn_hood_horn_e_base_l, swing1LimitDegrees=5, swing2LimitDegrees=5, twistLimitDegrees=15}
 {bone1Name=dyn_hood_horn_upMid, bone2Name=dyn_hood_horn_upMid_base, swing1LimitDegrees=5, swing2LimitDegrees=5, twistLimitDegrees=15}
 {bone1Name=dyn_hood_horn_e_r, bone2Name=dyn_hood_horn_e_base_r, swing1LimitDegrees=5, swing2LimitDegrees=5, twistLimitDegrees=15}
 {bone1Name=dyn_hood_horn_d_r, bone2Name=dyn_hood_horn_d_base_r, swing1LimitDegrees=5, swing2LimitDegrees=5, twistLimitDegrees=15}
 {bone1Name=dyn_hood_horn_c_r, bone2Name=dyn_hood_horn_c_base_r, swing1LimitDegrees=10, swing2LimitDegrees=5, twistLimitDegrees=15}
 {bone1Name=dyn_hood_horn_b_r, bone2Name=dyn_hood_horn_b_base_r, swing1LimitDegrees=5, swing2LimitDegrees=5, twistLimitDegrees=15}
 {bone1Name=dyn_hood_horn_a_r, bone2Name=dyn_hood_horn_a_base_r, swing1LimitDegrees=5, swing2LimitDegrees=5, twistLimitDegrees=15}
]
