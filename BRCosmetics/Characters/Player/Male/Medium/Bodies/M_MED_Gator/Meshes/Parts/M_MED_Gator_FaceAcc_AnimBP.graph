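# Post-copy Animation Blueprint for the Gator player body's face accessory mesh, adding hood horn motion on top of the owner mesh's pose.
# Weighted constraints bind horn base bones to head, neck, spine and clavicles.
# Above LOD 1, a flag set each update also constrains upper-arm twist bones to their upper arms.
# A final RigidBody node simulates the accessory physics asset in spine_05 space, with damped component motion and overridden gravity.
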AnimBlueprint /BRCosmetics/Characters/Player/Male/Medium/Bodies/M_MED_Gator/Meshes/Parts/M_MED_Gator_FaceAcc_AnimBP
target_skeleton=&/M_MED_Gator_FaceAcc_Skeleton
graphs [
 AnimGraph {nodes [
   n1 = Get MeshToCopy() -> (MeshToCopy:Skeletal Mesh Component Object Reference)
   n2 = "Constraint - dyn_hood_horn_a_base_l (PPPPPPP)"!(
    ComponentPose=Local To Component(LocalPose=Copy Pose From Mesh(SourceMeshComponent=@n1))
    Alpha=1
    ConstraintWeights_0=0.4583
    ConstraintWeights_1:float=0.022
    ConstraintWeights_2:float=0.0158
    ConstraintWeights_3:float=0.0547
    ConstraintWeights_4:float=0.1376
    ConstraintWeights_5:float=0.2809
    ConstraintWeights_6:float=0.0307
   ) -> (Pose) {
    alphaBoolBlend {blendOption=Linear}
    boneToModify {boneName=dyn_hood_horn_a_base_l}
    constraintSetup [
     {targetBone {boneName=head}}
     {targetBone {boneName=clavicle_l}}
     {targetBone {boneName=spine_02}}
     {targetBone {boneName=spine_03}}
     {targetBone {boneName=spine_04}}
     {targetBone {boneName=spine_05}}
     {targetBone {boneName=neck_02}}
    ]
    constraintWeights [1,1,1,1,1,1,1]
   }
   n3 = "Constraint - dyn_hood_horn_b_base_l (PPPPPP)"!(
    ComponentPose=@n2
    Alpha=1
    ConstraintWeights_0=0.8089
    ConstraintWeights_1:float=0.0132
    ConstraintWeights_2:float=0.0137
    ConstraintWeights_3:float=0.0364
    ConstraintWeights_4:float=0.1093
    ConstraintWeights_5:float=0.0186
   ) -> (Pose) {
    alphaBoolBlend {blendOption=Linear}
    boneToModify {boneName=dyn_hood_horn_b_base_l}
    constraintSetup [
     {targetBone {boneName=head}}
     {targetBone {boneName=clavicle_l}}
     {targetBone {boneName=spine_03}}
     {targetBone {boneName=spine_04}}
     {targetBone {boneName=spine_05}}
     {targetBone {boneName=neck_02}}
    ]
    constraintWeights [1,1,1,1,1,1]
   }
   n4 = "Constraint - dyn_hood_horn_c_base_l (PP)"!(ComponentPose=@n3, Alpha=1, ConstraintWeights_0=0.9862, ConstraintWeights_1:float=0.0138) -> (Pose) {
    alphaBoolBlend {blendOption=Linear}
    boneToModify {boneName=dyn_hood_horn_c_base_l}
    constraintSetup [{targetBone {boneName=head}}, {targetBone {boneName=spine_05}}]
    constraintWeights [1,1]
   }
   n5 = "Constraint - dyn_hood_horn_d_base_l (P)"!(ComponentPose=@n4, Alpha=1, ConstraintWeights_0=1, ConstraintWeights_1:float=0) -> (Pose) {
    alphaBoolBlend {blendOption=Linear}
    boneToModify {boneName=dyn_hood_horn_d_base_l}
    constraintSetup [{targetBone {boneName=head}}]
    constraintWeights [1,0]
   }
   n6 = "Constraint - dyn_hood_horn_e_base_l (P)"!(ComponentPose=@n5, Alpha=1, ConstraintWeights_0=1, ConstraintWeights_1:float=0) -> (Pose) {
    alphaBoolBlend {blendOption=Linear}
    boneToModify {boneName=dyn_hood_horn_e_base_l}
    constraintSetup [{targetBone {boneName=head}}]
    constraintWeights [1,0]
   }
   n7 = "Constraint - dyn_hood_horn_upMid_base (P)"!(ComponentPose=@n6, Alpha=1, ConstraintWeights_0=1, ConstraintWeights_1:float=0) -> (Pose) {
    alphaBoolBlend {blendOption=Linear}
    boneToModify {boneName=dyn_hood_horn_upMid_base}
    constraintSetup [{targetBone {boneName=head}}]
    constraintWeights [1,0]
   }
   n8 = "Constraint - dyn_hood_horn_a_base_r (PPPPPPP)"!(
    ComponentPose=@n7
    Alpha=1
    ConstraintWeights_0=0.4981
    ConstraintWeights_1:float=0.0142
    ConstraintWeights_2:float=0.0504
    ConstraintWeights_3:float=0.1271
    ConstraintWeights_4:float=0.2621
    ConstraintWeights_5:float=0.0186
    ConstraintWeights_6:float=0.0294
   ) -> (Pose) {
    alphaBoolBlend {blendOption=Linear}
    boneToModify {boneName=dyn_hood_horn_a_base_r}
    constraintSetup [
     {targetBone {boneName=head}}
     {targetBone {boneName=spine_02}}
     {targetBone {boneName=spine_03}}
     {targetBone {boneName=spine_04}}
     {targetBone {boneName=spine_05}}
     {targetBone {boneName=clavicle_r}}
     {targetBone {boneName=neck_02}}
    ]
    constraintWeights [1,1,1,1,1,1,1]
   }
   n9 = "Constraint - dyn_hood_horn_b_base_r (PPPPPP)"!(
    ComponentPose=@n8
    Alpha=1
    ConstraintWeights_0=0.8436
    ConstraintWeights_1:float=0.0106
    ConstraintWeights_2:float=0.0298
    ConstraintWeights_3:float=0.0933
    ConstraintWeights_4:float=0.006
    ConstraintWeights_5:float=0.0166
   ) -> (Pose) {
    alphaBoolBlend {blendOption=Linear}
    boneToModify {boneName=dyn_hood_horn_b_base_r}
    constraintSetup [
     {targetBone {boneName=head}}
     {targetBone {boneName=spine_03}}
     {targetBone {boneName=spine_04}}
     {targetBone {boneName=spine_05}}
     {targetBone {boneName=clavicle_r}}
     {targetBone {boneName=neck_02}}
    ]
    constraintWeights [1,1,1,1,1,1]
   }
   n10 = "Constraint - dyn_hood_horn_c_base_r (PP)"!(ComponentPose=@n9, Alpha=1, ConstraintWeights_0=0.9939, ConstraintWeights_1:float=0.0061) -> (Pose) {
    alphaBoolBlend {blendOption=Linear}
    boneToModify {boneName=dyn_hood_horn_c_base_r}
    constraintSetup [{targetBone {boneName=head}}, {targetBone {boneName=spine_05}}]
    constraintWeights [1,1]
   }
   n11 = "Constraint - dyn_hood_horn_d_base_r (P)"!(ComponentPose=@n10, Alpha=1, ConstraintWeights_0=1, ConstraintWeights_1:float=0) -> (Pose) {
    alphaBoolBlend {blendOption=Linear}
    boneToModify {boneName=dyn_hood_horn_d_base_r}
    constraintSetup [{targetBone {boneName=head}}]
    constraintWeights [1,0]
   }
   n12 = "Constraint - dyn_hood_horn_e_base_r (P)"!(ComponentPose=@n11, Alpha=1, ConstraintWeights_0=1, ConstraintWeights_1:float=0) -> (Pose) {
    alphaBoolBlend {blendOption=Linear}
    boneToModify {boneName=dyn_hood_horn_e_base_r}
    constraintSetup [{targetBone {boneName=head}}]
    constraintWeights [1,0]
   }
   n13 = Get isLOD2Over() -> (isLOD2Over:bool)
   n14 = "Constraint - upperarm_twist_01_l (P)"!(ComponentPose=@n12, bAlphaBoolEnabled:bool=@n13, ConstraintWeights_0=1) -> (Pose) {
    alphaInputType=Bool
    alphaBoolBlend {blendOption=Linear}
    boneToModify {boneName=upperarm_twist_01_l}
    constraintSetup [{targetBone {boneName=upperarm_l}}]
    constraintWeights [1]
   }
   n15 = "Constraint - upperarm_twist_02_l (P)"!(ComponentPose=@n14, bAlphaBoolEnabled:bool=@n13, ConstraintWeights_0=1) -> (Pose) {
    alphaInputType=Bool
    alphaBoolBlend {blendOption=Linear}
    boneToModify {boneName=upperarm_twist_02_l}
    constraintSetup [{targetBone {boneName=upperarm_l}}]
    constraintWeights [1]
   }
   n16 = "Constraint - upperarm_twist_01_r (P)"!(ComponentPose=@n15, bAlphaBoolEnabled:bool=@n13, ConstraintWeights_0=1) -> (Pose) {
    alphaInputType=Bool
    alphaBoolBlend {blendOption=Linear}
    boneToModify {boneName=upperarm_twist_01_r}
    constraintSetup [{targetBone {boneName=upperarm_r}}]
    constraintWeights [1]
   }
   n17 = "Constraint - upperarm_twist_02_r (P)"!(ComponentPose=@n16, bAlphaBoolEnabled:bool=@n13, ConstraintWeights_0=1) -> (Pose) {
    alphaInputType=Bool
    alphaBoolBlend {blendOption=Linear}
    boneToModify {boneName=upperarm_twist_02_r}
    constraintSetup [{targetBone {boneName=upperarm_r}}]
    constraintWeights [1]
   }
   n18 = RigidBody(ComponentPose=@n17) {
    bClampLinearTranslationLimitToRefPose=true
    bOverrideWorldGravity=true
    overridePhysicsAsset=&/M_MED_Gator_FaceAcc_Physics
    simulationSpace=BaseBoneSpace
    alphaBoolBlend {blendOption=Linear}
    baseBoneRef {boneName=spine_05}
    componentLinearAccScale=<0.4 0.4 0.4>
    componentLinearVelScale=<0.4 0.4 0.4>
    overrideWorldGravity {z=-980}
   }
   Output Pose(Result=Component To Local(ComponentPose=@n18)) {layerGroup=None}
  ]}
 EventGraph {nodes [
   Event BlueprintUpdateAnimation()
   > n1 = Branch(Condition="integer > integer"(A=Get CurrentLOD(), B=1) {is=title K2Node_CallFunction})
   > Set isLOD2Over(isLOD2Over:bool=true) -> (Output_Get:bool)
   Set isLOD2Over(execute=@n1.else, isLOD2Over=false)
   Event BlueprintInitializeAnimation()
   > Set MeshToCopy(MeshToCopy:Skeletal Mesh Component Object Reference=GetOwnerSkeletalMeshComponent()) -> (Output_Get:Skeletal Mesh Component Object Reference)
   n2 = TryGetPawnOwner()
   > Is Valid(InputObject=@n2)
   > n3 = Cast To FortPlayerPawnAthena(Object=@n2)
   > Set playerPawn(playerPawn:Fort Player Pawn Athena Object Reference=@n3) -> (Output_Get:Fort Player Pawn Athena Object Reference)
   TryGetPawnOwner()
  ]}
]
variables [
 isLOD2Over {type=boolean}
 MeshToCopy {type=SkeletalMeshComponent}
 playerPawn {type=FortPlayerPawnAthena}
]
